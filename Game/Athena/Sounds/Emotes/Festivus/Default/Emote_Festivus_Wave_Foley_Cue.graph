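# Sound cue for the foley of the Festivus Wave emote in Athena, built from Commando step and foley waves.
# Mixes an immediate random step with a briefly delayed second step and a quieter, delayed, pitch-varied foley rustle under a root modulator.
# Quality-level nodes supply both delayed layers only in the first two of four slots.
# Routes through emote music sound class and concurrency, with emote attenuation and a music-analysis submix send.
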SoundCue /Game/Athena/Sounds/Emotes/Festivus/Default/Emote_Festivus_Wave_Foley_Cue
root=Modulator_0
nodes [
 {id=Delay_0, children [Random_1], properties {delayMax=0.09, delayMin=0.05}}
 {id=Delay_1, children [Modulator_1], properties {delayMax=0.15, delayMin=0.05}}
 {id=Mixer_0, children [Random_0, QualityLevel_1, QualityLevel_0], properties {inputVolume [1,1,0.3]}}
 {id=Modulator_0, children [Mixer_0]}
 {id=Modulator_1, children [Random_2], properties {pitchMax=1.2, pitchMin=0.8}}
 {id=QualityLevel_0, children [Delay_1,Delay_1,null,null]}
 {id=QualityLevel_1, children [Delay_0,Delay_0,null,null]}
 {id=Random_0, children [WavePlayer_10, WavePlayer_11, WavePlayer_12], properties {bRandomizeWithoutReplacement=false, weights [1,1,1]}}
 {id=Random_1, children [WavePlayer_13, WavePlayer_14], properties {bRandomizeWithoutReplacement=false, weights [1,1]}}
 {id=Random_2, children [WavePlayer_15, WavePlayer_16], properties {weights [1,1]}}
 {id=WavePlayer_10, properties {soundWaveAssetPtr=/Game/Sounds/Fortress_Foley/Commando/Commando_Step_001}}
 {id=WavePlayer_11, properties {soundWaveAssetPtr=/Game/Sounds/Fortress_Foley/Commando/Commando_Step_002}}
 {id=WavePlayer_12, properties {soundWaveAssetPtr=/Game/Sounds/Fortress_Foley/Commando/Commando_Step_003}}
 {id=WavePlayer_13, properties {soundWaveAssetPtr=/Game/Sounds/Fortress_Foley/Commando/Commando_Step_004}}
 {id=WavePlayer_14, properties {soundWaveAssetPtr=/Game/Sounds/Fortress_Foley/Commando/Commando_Step_005}}
 {id=WavePlayer_15, properties {soundWaveAssetPtr=/Game/Sounds/Fortress_Foley/Commando/Commando_Foley_Stereo_001}}
 {id=WavePlayer_16, properties {soundWaveAssetPtr=/Game/Sounds/Fortress_Foley/Commando/Commando_Foley_Stereo_003}}
]
properties {
 attenuationSettings=&Fort_Emotes_Athena_Att
 maxDistance=1600
 soundClassObject=&Fort_Music_Emote
 volumeMultiplier=0.5
 concurrencySet [/Game/Sounds/Concurrency_Rules/Athena_EmoteMusic_Conc]
 soundSubmixSends [{minSendLevel=0, soundSubmix=&Fortnite_AudioAnalysis_Music}]
}
